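# Animation Blueprint for the LoudPhoenix emote prop mesh, built on the LoudPhoenix_Emote skeleton.
# It copies the pose from the attached parent mesh as the base of a layered blend per bone.
# DefaultSlot, with no source pose wired in, feeds the layer filtered from lowerarm_r, so slot montages drive that right-arm branch.
# The event graph holds only unconnected default nodes, and no variables are defined.
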AnimBlueprint /BRCosmetics/Animation/Props/Emote_LoudPhoenix/Meshes/LoudPhoenix_Emote_AnimBP
target_skeleton=&/LoudPhoenix_Emote_Skeleton
graphs [
 AnimGraph {nodes [Output Pose(Result=Layered blend per bone(BasePose=Copy Pose From Mesh() {bUseAttachedParent=true}, BlendPoses_0=Slot 'DefaultSlot'()) {layerSetup [{branchFilters [{boneName=lowerarm_r}]}]}) {layerGroup=None}]}
 EventGraph {nodes [
   TryGetPawnOwner()
   Event BlueprintUpdateAnimation()
  ]}
]
variables []
